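; Shared vehicle cosmetics function library that configures Niagara drift trail and boost effects.
; Painted trails look up colors in the Painted DataTable, let Additional Info colors override them, and push float and int parameters; a helper maps color names to one-based row indices.
; Boost helpers set position from socket deltas against the attach parent and per-body offsets from the TurboOffsets DataTable.
; Integer parameters also select materials from an array for Niagara material variables.
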(fn BP_GetNSIndexByColor
  (params (Color :type Name))
  (bind _outrownames (DataTable|GetDataTableRowNames "/VehicleCosmetics/DataTables/DT_VehicleCosmetics_Painted.DT_VehicleCosmetics_Painted"))
  (bind (_array_element _array_index) (Utilities|Array|ForEachLoopwithBreak _outrownames)
    (:LoopBody
      (if (== _array_element Color)
        (LocalVariables|SetIndex (+ _array_index 1)))
      (LocalVariables|SetIndex 0))
    (:Completed
      _))
  (return (LocalVariables|GetIndex)))

(fn BP_SetPaintedDriftTrailBase
  (params (Color :type Name :default "None") (Drift%20Trail :type Niagara%20Particle%20System%20Component%20Object%20Reference) (Variable%20Name%20Bool%20Painted :type Name :default "IsPainted") (Variable%20Names%20Color%20Painted :type Array%20of%20Names) (Additional%20Info :type Vehicle%20Cosmetics%20Additional%20Parameter%20Infos%20Structure))
  (bind (_floatparameters _intparameters _colorparameters) (Utilities|Struct|BreakVehicleCosmeticsAdditionalParameterInfos Additional%20Info))
  (Utilities|FlowControl|Sequence
    (:then_0
      (Utilities|GetDataTableRow "/VehicleCosmetics/DataTables/DT_VehicleCosmetics_Painted.DT_VehicleCosmetics_Painted" Color
        (:then
          (Niagara|SetNiagaraVariable(Bool) Drift%20Trail Variable%20Name%20Bool%20Painted true)
          (LocalVariables|SetPainted? true))
        (:RowNotFound
          (Niagara|SetNiagaraVariable(Bool) Drift%20Trail Variable%20Name%20Bool%20Painted false)
          (LocalVariables|SetPainted? false))))
    (:then_1
      (if (> (Utilities|Array|Length _colorparameters) 0)
        (for _array_element _colorparameters
          (Niagara|SetNiagaraVariable(LinearColor) Drift%20Trail (Utilities|Struct|BreakVehicleCosmeticsCustomizableObjectColorInfo _array_element) (.ParameterValue (Utilities|Struct|BreakVehicleCosmeticsCustomizableObjectColorInfo _array_element))))
        (else
          (bind _returnvalue (Utilities|GetDataTableRow "/VehicleCosmetics/DataTables/DT_VehicleCosmetics_Painted.DT_VehicleCosmetics_Painted" Color))
          (for _array_element_2 Variable%20Names%20Color%20Painted
            (Niagara|SetNiagaraVariable(LinearColor) Drift%20Trail _array_element_2 (Utilities|Struct|BreakStructVehicleCosmeticsPainted _returnvalue))))))
    (:then_2
      (for _array_element_3 _floatparameters
        (Niagara|SetNiagaraVariable(Float) Drift%20Trail (Utilities|Struct|BreakVehicleCosmeticsCustomizableObjectFloatInfo _array_element_3) (.ParameterValue (Utilities|Struct|BreakVehicleCosmeticsCustomizableObjectFloatInfo _array_element_3)))))
    (:then_3
      (for _array_element_4 _intparameters
        (Niagara|SetNiagaraVariable(Int32) Drift%20Trail (Utilities|Struct|BreakVehicleCosmeticsCustomizableObjectIntInfo _array_element_4) (Utilities|String|StringToInteger (Utilities|String|ToString(Name) (.ParameterValue (Utilities|Struct|BreakVehicleCosmeticsCustomizableObjectIntInfo _array_element_4)))))
        (return (LocalVariables|GetPainted?))))))

(fn BP_SingleBoostSetPosition
  (params (Boost :type Niagara%20Particle%20System%20Component%20Object%20Reference) (Variable%20Name%20Position :type Name :default "SingleBoostPosition"))
  (bind (_returnvalue_x _returnvalue_y _returnvalue_z) (Math|Vector|BreakVector (Transformation|GetSocketLocation Boost "Socket_Turbo_A_R")))
  (bind (_returnvalue_x_2 _returnvalue_y_2 _returnvalue_z_2) (Math|Vector|BreakVector (Transformation|GetSocketLocation (Transformation|GetAttachParent Boost) "Socket_Turbo_A_R")))
  (Niagara|SetNiagaraVariable(Vector3) Boost Variable%20Name%20Position (Math|Vector|MakeVector (Utilities|Operators|Subtract _returnvalue_x_2 _returnvalue_x 0) 0.0 (- _returnvalue_z_2 _returnvalue_z))))

(fn SetBoostOffsetByCar
  (params (Vehicle%20Cosmetics%20Interface :type Vehicle%20Cosmetics%20Interface%20Interface) (Boost :type Niagara%20Particle%20System%20Component%20Object%20Reference) (Offset%20Variable%20Name :type Name :default "DTCarOffset"))
  (bind _returnvalue (Vehicle|Cosmetics|GetBodyItemName Vehicle%20Cosmetics%20Interface))
  (if (Utilities|IsValidInterface Vehicle%20Cosmetics%20Interface)
    (bind _returnvalue_2 (Utilities|GetDataTableRow "/VehicleCosmetics/DataTables/DT_VehicleCosmetics_TurboOffsets.DT_VehicleCosmetics_TurboOffsets" _returnvalue)
      (:then
        (Niagara|SetNiagaraVariable(Vector3) Boost Offset%20Variable%20Name (Utilities|Struct|BreakStructVehiclesCosmeticsTurboOffset _returnvalue_2)))
      (:RowNotFound
        (Development|PrintString 0 (Utilities|String|Append (Utilities|String|ToString(Name) _returnvalue) " - is not in the table") true true "(R=1.000000,G=0.678066,B=0.000000,A=1.000000)" 4.0)))
    (else
      (Development|PrintString 0 "Invalid interface" true true "(R=1.000000,G=0.068987,B=0.000000,A=1.000000)" 3.0)))
  (return))

(fn SetMICBasedOnValue
  (params (Additional%20Info :type Vehicle%20Cosmetics%20Additional%20Parameter%20Infos%20Structure) (NiagaraSystem :type Niagara%20Particle%20System%20Component%20Object%20Reference) (MaterialsArray :type Array%20of%20Material%20Interface%20Object%20References) (Variables%20Name :type Array%20of%20Struct%20Vehicles%20Cosmetics%20MISelector%20Structures))
  (bind (_floatparameters _intparameters) (Utilities|Struct|BreakVehicleCosmeticsAdditionalParameterInfos Additional%20Info))
  (Utilities|IsValid NiagaraSystem
    (:Is%20Valid
      (if (> (Utilities|Array|Length _intparameters) 0)
        (bind _array_element (Utilities|Array|ForEachLoopwithBreak _intparameters)
          (:LoopBody
            (for _array_element_2 Variables%20Name
              (if (== (Utilities|Struct|BreakVehicleCosmeticsCustomizableObjectIntInfo _array_element) (Utilities|Struct|BreakStructVehiclesCosmeticsMISelector _array_element_2))
                (if (Utilities|Array|IsValidIndex MaterialsArray (Utilities|String|StringToInteger (Utilities|String|ToString(Name) (.ParameterValue (Utilities|Struct|BreakVehicleCosmeticsCustomizableObjectIntInfo _array_element)))))
                  (Niagara|SetNiagaraVariable(Material) NiagaraSystem (.NS_ParamName_4_575D67F547CE24A1685C22952F38ABAF (Utilities|Struct|BreakStructVehiclesCosmeticsMISelector _array_element_2)) (Utilities|Array|Get(acopy) MaterialsArray (Utilities|String|StringToInteger (Utilities|String|ToString(Name) (.ParameterValue (Utilities|Struct|BreakVehicleCosmeticsCustomizableObjectIntInfo _array_element))))))
                  (else
                    (Development|PrintString 0 "Invalid Mic Value" true true "(R=1.000000,G=0.886706,B=0.000000,A=1.000000)" 4.0))))))
          (:Completed
            _)))
      (return))))
