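; Shared vehicle cosmetics function library whose ShouldTriggerFX decides whether a vehicle's cosmetic effects play.
; It gets the vehicle actor from the cosmetics interface, casts it to FortAthenaVehicle and compares its significance LOD against a maximum allowed LOD, which the caller can override.
; The fallback result is true, so effects play by default when that LOD check cannot be made.
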(fn ShouldTriggerFX
  (params (VehicleInterface :type Vehicle%20Cosmetics%20Interface%20Interface) (MaxLODAllowed :type Integer :default 1))
  (if (Utilities|IsValidInterface VehicleInterface)
    (bind _returnvalue (Vehicle|Cosmetics|GetVehicleActor VehicleInterface))
    (bind _asfort_athena_vehicle (Utilities|Casting|CastToFortAthenaVehicle _returnvalue)
      (:then
        (return (<= (Vehicle|GetSignificanceLOD _asfort_athena_vehicle) MaxLODAllowed)))
      (:CastFailed
        _)))
  ;[1 1 1] Note: Default to true, i.e. Play FX
  (return true))
